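# Material is a Default Lit surface material in the Radiant "guy" mesh folder, giving a flat orange look.
# It comes from an Interchange import and holds constant Phong inputs: black ambient, orange diffuse, dimmer orange specular and a shininess value.
# These feed MF_PhongToMetalRoughness, which converts Phong shading to Unreal's metal-roughness model, with no textures or exposed parameters.
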Material /Game/NeverCook/FM/Radiant/Meshes/guy/Material
graph {
 properties {assetImportData="<self>:InterchangeAssetImportData_0", usedShadingModels=Default Lit}
 nodes [MaterialFunctionCall!(
   AmbientColor=Constant3Vector() {constant=<rgba 0 0 0 1>}
   DiffuseColor=Constant3Vector() {constant=<rgba 1 0.4659 0 1>}
   SpecularColor=Constant3Vector() {constant=<rgba 0.25 0.1165 0 1>}
   Shininess=Constant() {r=96.04}
  ) {materialFunction=/InterchangeAssets/Functions/MF_PhongToMetalRoughness}]
}
